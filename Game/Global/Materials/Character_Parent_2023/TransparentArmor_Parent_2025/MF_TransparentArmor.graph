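# Material function that restyles incoming character material attributes as transparent armor, optionally blended in through a TechArtMask channel.
# A custom HLSL spiral blur samples refracted scene color, with quality-scaled steps, to fake see-through material.
# Sky-light refraction gradients, Fresnel thin-film iridescence and micro normals add surface detail, with emissive scaled by time of day.
# Feature-level switches substitute simpler base color, emissive and opacity on low-end platforms.
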MaterialFunction /Game/Global/Materials/Character_Parent_2023/TransparentArmor_Parent_2025/MF_TransparentArmor
graph {nodes [
  MaterialExpressionViewSize_2 = ViewSize()
  n1 = Convert(None=@MaterialExpressionViewSize_2) {
   nodeName=Break Float2
   convertInputs [{
     type=Vector2
     expressionInput {
      expression=&self:MaterialExpressionViewSize_2
      inputName=None
      mask=0
      maskA=0
      maskB=0
      maskG=0
      maskR=0
      outputIndex=0
     }
    }]
   convertOutputs [{type=Scalar}, {type=Scalar}, {type=Vector2}]
  }
  ComponentMask(None=MaterialFunctionCall(Texture Dimensions=@n1) {materialFunction=&ScreenAlignedUVs}."X 100%, Y 100%") {g=true, r=true}
  n2 = SkyAtmosphereLightDirection()
  n3 = Transform(None=NamedRerouteUsage())
  n4 = MaterialFunctionCall!(
   Ray Direction=@n2
   Surface Normal=@n3
   Refractive Index Origin
   Refractive Index Target=ScalarParameter() {defaultValue=1, group=TransparentArmor, parameterName=Refraction IOR}
  ) {materialFunction=/Engine/Functions/Engine_MaterialFunctions01/Vectors/Refract}
  ComponentMask(None=@n4.Vector) {b=true, r=true}
  Constant() {r=1}
  Constant() {r=1}
  n5 = NamedRerouteUsage()
  n6 = SetMaterialAttributes!(
   MaterialAttributes
   BaseColor=NamedRerouteUsage()
   Metallic=NamedRerouteUsage()
   Specular=NamedRerouteUsage()
   Roughness=NamedRerouteUsage()
   EmissiveColor=NamedRerouteUsage()
   Opacity=NamedRerouteUsage()
   Normal=NamedRerouteUsage()
  ) {attributeSetTypes [
    #1
    #2
    #3
    #4
    #5
    #6
    #7
   ]}
  n7 = ChannelMaskParameter(None=NamedRerouteUsage()) {parameterName=TransparentArmorChannel}
  FunctionOutput(None=StaticSwitchParameter(True=BlendMaterialAttributes(A=@n5, B=@n6, Alpha=@n7.R), False=@n6) {group=TransparentArmor, parameterName=UseTechArtMaskforTransparentArmor, sortPriority=0})
  n8 = TextureSampleParameter2D(UVs=Multiply(A=TextureCoordinate(), B=ScalarParameter() {defaultValue=100, group=TransparentArmor, parameterName=MicroNormalTiling}) {constB=200}) {
   group=TransparentArmor
   parameterName=MicroNormal
   samplerType=SAMPLERTYPE_LinearColor
   texture=/Game/TimeOfDay/CloudsAndStorm/Materials/M_MakeNoise_Baked
  }
  NamedRerouteDeclaration(None=@n8.RGB) {name=MicroNormal, nodeColor=<rgba 1 0.75 0 1>}
  n9 = TextureSampleParameter2D() {
   group=TransparentArmor
   parameterName=Normal
   samplerSource=SSM_Wrap_WorldGroupSettings
   samplerType=SAMPLERTYPE_Normal
   texture=&FlatNormalBlue
  }
  NamedRerouteDeclaration(None=@n9.RGB) {name=Normal_In, nodeColor=<rgba 0.4297 0 1 1>}
  n10 = GetMaterialAttributes(None=FunctionInput() {inputType=FunctionInput_MaterialAttributes}) {attributeGetTypes [
    #1
    #2
    #3
    #4
    #5
   ]}
  NamedRerouteDeclaration(None=@n10.BaseColor) {name=BaseColor_In, nodeColor=<rgba 0 0.8828 1 1>}
  NamedRerouteDeclaration(None=StaticSwitchParameter(True=NamedRerouteUsage(), False=ScalarParameter() {defaultValue=0.2, group=TransparentArmor, parameterName=Roughness, sortPriority=1}) {group=TransparentArmor, parameterName=-UseOriginalRoughness, sortPriority=1}) {name=Roughness, nodeColor=<rgba 1 0.4922 0 1>}
  NamedRerouteDeclaration(None=StaticSwitchParameter(True=NamedRerouteUsage(), False=ScalarParameter() {group=TransparentArmor, parameterName=Metallic, sortPriority=1}) {group=TransparentArmor, parameterName=-UseOriginalMetallic, sortPriority=1}) {name=Metallic, nodeColor=<rgba 1 0.4922 0 1>}
  NamedRerouteDeclaration(None=@n10.Metallic) {name=Metallic_In, nodeColor=<rgba 0 0.8828 1 1>}
  NamedRerouteDeclaration(None=@n10.Specular) {name=Specular_In, nodeColor=<rgba 0 0.8828 1 1>}
  NamedRerouteDeclaration(None=@n10.Roughness) {name=Roughness_In, nodeColor=<rgba 0 0.8828 1 1>}
  NamedRerouteDeclaration(None=@n10.EmissiveColor) {name=Emissive_In, nodeColor=<rgba 1 0 0.1641 1>}
  NamedRerouteDeclaration(None=FunctionInput() {inputName=TechArtMask}) {name=TechArtMask, nodeColor=<rgba 0.2422 1 0 1>}
  NamedRerouteDeclaration(None=MaterialFunctionCall(X=Add(A=@n8.R, B=@n9.R), Y=Add(A=@n8.G, B=@n9.G), Z=Multiply(A=@n8.B, B=@n9.B)) {materialFunction=&MakeFloat3}.Result) {name=Normal_Out, nodeColor=<rgba 0 1 0.1094 1>}
  NamedRerouteDeclaration(None=@n10.MaterialAttributes) {name=Mat_In, nodeColor=<rgba 1 0.1406 0 1>}
  n11 = TextureSampleParameter2D(UVs=MaterialFunctionCall!(Vector=@n4.Vector) {materialFunction=/Engine/Functions/Engine_MaterialFunctions01/Coordinates/LongLatToUV}.Result) {
   group=TransparentArmor
   parameterName=RefractGradient
   samplerSource=SSM_Wrap_WorldGroupSettings
   texture=/Game/Characters/Player/Female/Medium/Bodies/F_MED_Critter_Streak/Textures/CritterStreak_Gradient
  }
  n12 = DotProduct(A=@n2, B=@n3)
  n13 = TextureSampleParameter2D(UVs=@n12) {
   group=TransparentArmor
   parameterName=RefractGradient
   samplerSource=SSM_Wrap_WorldGroupSettings
   texture=&CritterStreak_Gradient
  }
  n14 = ScalarParameter() {defaultValue=1, group=TransparentArmor, parameterName=SdN_Pow}
  n15 = Multiply(
   A=@n13.RGB
   B=Multiply(A=Power(Base=@n12, Exp=@n14), B=ScalarParameter() {defaultValue=10, group=TransparentArmor, parameterName=SdN_Mult}) {constB=0.5}
  )
  NamedRerouteDeclaration(None=LinearInterpolate(
    A=ScalarParameter() {group=TransparentArmor, parameterName=FakeHighLight_Low}
    B=ScalarParameter() {defaultValue=1, group=TransparentArmor, parameterName=FakeHighLight_High}
    Alpha=Max(A=@n11.RGB, B=@n15)
   )) {name=FakeLighting, nodeColor=<rgba 1 0 0 1>}
  n16 = Multiply(A=NamedRerouteUsage(), B=ScalarParameter() {defaultValue=1, group=TransparentArmor, parameterName=EmissiveMult, sortPriority=3})
  n17 = VectorParameter() {group=TransparentArmor, parameterName=SurfaceEmissiveColor, sortPriority=3, defaultValue=<rgba 0.7588 0.974 0.9553 1>}
  n18 = Transform(None=NamedRerouteUsage())
  n19 = Multiply(
   A=NamedRerouteUsage()
   B=LinearInterpolate(
    A=StaticSwitchParameter(True=@n16, False=@n17.RGB) {group=TransparentArmor, parameterName=-UseTextureMapforEmissive, sortPriority=3}
    Alpha=Fresnel(Normal=@n18) {exponent=0.2}
   ) {constB=0}
  )
  n20 = NamedRerouteUsage()
  n21 = ScalarParameter() {defaultValue=1, group=TransparentArmor, parameterName=Emissive_Final_Day}
  n22 = Multiply(
   A=Add(A=Add(A=@n19, B=@n20), B=NamedRerouteUsage())
   B=MaterialFunctionCall(
    Value_Morning=@n21
    Value_Day=@n21
    Value_Evening=@n21
    Value_Night=ScalarParameter() {defaultValue=0.5, group=TransparentArmor, parameterName=Emissive_Final_Night}
   ) {materialFunction=&Scalar_Time_of_Day}.Result
  )
  n23 = VectorParameter() {group=TransparentArmor, parameterName=SurfaceEmissiveColor_LowEnd, sortPriority=3, defaultValue=<rgba 0.0937 0.1002 0.1406 1>}
  n24 = StaticSwitchParameter(
   True=Multiply(A=NamedRerouteUsage(), B=ScalarParameter() {defaultValue=1, group=TransparentArmor, parameterName=EmissiveMult_LowEnd, sortPriority=3})
   False=@n23.RGB
  ) {group=TransparentArmor, parameterName=-UseTextureMapforEmissive, sortPriority=3}
  NamedRerouteDeclaration(None=FeatureLevelSwitch(Default=@n22, ES2=@n24, ES3_1=@n24, SM5=@n22, SM6=@n22)) {name=Emissive, nodeColor=<rgba 0.2656 0 1 1>}
  n25 = NamedRerouteUsage()
  n26 = MaterialFunctionCall!(Circumference Tiling, Position=LocalPosition().XYZ, Z Height, HQ Angle Calculation?) {materialFunction=/Game/Characters/Player/Common/MaterialFunctions/MF_CylindricalUVs_New}
  n27 = Multiply(A=Constant2Vector() {g=2, r=1}, B=ScalarParameter() {defaultValue=1, group=TransparentArmor, parameterName=CylindricalUVTiling})
  n28 = MaterialFunctionCall(
   Normal=TextureSample(UVs=Multiply(A=@n26.UVs, B=@n27)) {samplerType=SAMPLERTYPE_Normal, texture=/Game/Gadgets/Assets/Telephone_Booth/Textures/T_BlindsNormalNoise_N}.RGB
   Flatness=ScalarParameter() {defaultValue=0.75, group=TransparentArmor, parameterName=CylindricalTiledNormalFlatten}
  ) {materialFunction=&FlattenNormal}
  n29 = Fresnel(
   ExponentIn=ScalarParameter() {defaultValue=2, group=TransparentArmor, parameterName=ThinFilmExp}
   Normal=Transform(None=MaterialFunctionCall(BaseNormal=@n25, AdditionalNormal=@n28.Result) {materialFunction=&BlendAngleCorrectedNormals}.Result)
  ) {baseReflectFraction=0.1, exponent=3}
  n30 = TextureSampleParameter2D(UVs=@n29) {group=TransparentArmor, parameterName=ThinFilmTexture, texture=/Game/Global/Textures/Default/Thinfilm/T_ThinFilm_Spectrum_CoolColor}
  n31 = LinearInterpolate(B=@n30.RGB, Alpha=Multiply(A=@n29))
  n32 = ScalarParameter() {defaultValue=0.6, group=TransparentArmor, parameterName=ThinFilmDesaturation}
  NamedRerouteDeclaration(None=Multiply(A=Desaturation(None=@n31, Fraction=@n32), B=ScalarParameter() {defaultValue=1, group=TransparentArmor, parameterName=ThinFilmMult})) {name=ThinFilm, nodeColor=<rgba 0.7812 0 1 1>}
  MaterialExpressionLinearInterpolate_2 = LinearInterpolate(
   A=ScalarParameter() {defaultValue=0.01, group=TransparentArmor, parameterName=SpiralBlur_Distance_Near}
   B=ScalarParameter() {defaultValue=0.003, group=TransparentArmor, parameterName=SpiralBlur_Distance_Far}
   Alpha=MaterialFunctionCall(Fade Length=ScalarParameter() {defaultValue=256, group=TransparentArmor, parameterName=SpiralBlur_FadingLength}) {materialFunction=&CameraDepthFade}.Result
  )
  n33 = Constant()
  MaterialExpressionMaterialFunctionCall_2 = MaterialFunctionCall!(Low=@n33, High=Constant() {r=4}, Epic=Constant() {r=8}, --------, Mobile=@n33, Medium=@n33) {materialFunction=/FortniteBaseMaterials/Functions/MF_QualitySwitch_Scalar}
  MaterialExpressionScalarParameter_21 = ScalarParameter() {defaultValue=1, parameterName=SpiralBlur_RadialSteps}
  MaterialExpressionConstant_4 = Constant() {r=1}
  MaterialExpressionTextureObject_5 = TextureObject() {samplerType=SAMPLERTYPE_LinearGrayscale, texture=/Engine/EngineMaterials/Good64x64TilingNoiseHighFreq}
  MaterialExpressionScalarParameter_22 = ScalarParameter() {defaultValue=0.618, parameterName=SpiralBlur_RadialOffset}
  n34 = MaterialFunctionCall() {materialFunction=&ObjectPivotPoint}
  n35 = CameraPositionWS()
  MaterialExpressionMultiply_23 = Multiply(
   A=ScalarParameter() {defaultValue=0.02, parameterName=SpiralBlur_RefractionAmount}
   B=Normalize(VectorInput=Subtract(A=@n34.Object Pivot Location, B=@n35))
  )
  n36 = Custom!(
   Distance=@MaterialExpressionLinearInterpolate_2
   DistanceSteps=@MaterialExpressionMaterialFunctionCall_2.Result
   RadialSteps=@MaterialExpressionScalarParameter_21
   TempAARotation=@MaterialExpressionConstant_4
   DistanceMask=@MaterialExpressionConstant_4
   Tex=@MaterialExpressionTextureObject_5
   TempAADistance=@MaterialExpressionConstant_4
   RadialOffset=@MaterialExpressionScalarParameter_22
   RefractionAmount=@MaterialExpressionMultiply_23
  ) {
   code="float3 CurColor = 0;\r\n\r\nfloat2 RawUV = MaterialFloat2(ScreenAlignedPosition(Parameters.ScreenPosition).xy);\r\n//float2 RefractedUV = BaseUV + (BaseUV.xy - 0.5) * RefractionAmount;\r\nfloat2 BaseUV = RawUV + (RawUV.xy - 0.5) * RefractionAmount;\r\nfloat2 NewUV = BaseUV;\r\nfloat StepSize = Distance / (int) DistanceSteps;\r\nfloat CurDistance = 0;\r\nfloat2 CurOffset = 0;\r\nfloat TwoPi = 6.283185;\r\nfloat Substep = 0;\r\nfloat2 ScenePixels=View.BufferSizeAndInvSize.xy*BaseUV;\r\nScenePixels+=View.TemporalAAParams.r;\r\nfloat2 RandomSamp = ((uint)(ScenePixels.x) + 2 * (uint)(ScenePixels.y)) % 5;\r\nRandomSamp+=Texture2DSample(Tex,TexSampler,ScenePixels);\r\nRandomSamp/=5;\r\nRandomSamp-=0.5;\r\nTempAARotation*=RandomSamp;\r\nTempAADistance*=StepSize*RandomSamp;\r\n\r\n\r\nint i=0;\r\nif (DistanceSteps<1)\r\n{\r\nreturn DecodeSceneColorForMaterialNode(NewUV);\r\n}\r\nelse\r\n{\r\nfloat2 BilinearOffset = View.BufferSizeAndInvSize.zw * 0.5f;\r\nfloat2 BilinearMinBufferUV = ViewportUVToBufferUV(float2(0.0f, 0.0f)) + BilinearOffset;\r\nfloat2 BilinearMaxBufferUV = ViewportUVToBufferUV(float2(1.0f, 1.0f)) - BilinearOffset;\r\n//CurDistance += 0.5*StepSize;\r\nwhile ( i < (int) DistanceSteps)\r\n{\r\n\r\n//CurDistance+=StepSize;\r\nfor (int j = 0; j < (int) RadialSteps; j++)\r\n{\r\nCurOffset.x = cos(TwoPi*((TempAARotation+Substep) / RadialSteps));\r\nCurOffset.y = sin(TwoPi*((TempAARotation+Substep) / RadialSteps));\r\nCurOffset *=DistanceMask;\r\nNewUV.x = BaseUV.x + (CurOffset.x * (CurDistance+(RandomSamp*TempAADistance)));\r\nNewUV.y = BaseUV.y + (CurOffset.y * (CurDistance+(RandomSamp*TempAADistance)));\r\nNewUV = clamp(NewUV, BilinearMinBufferUV, BilinearMaxBufferUV);\r\nCurColor += DecodeSceneColorForMaterialNode(NewUV);\r\n//CurDistance+=(StepSize+(TempAADistance))/RadialSteps;\r\nSubstep++;\r\n}\r\nCurDistance+=StepSize;\r\nSubstep+=RadialOffset;\r\ni++;\r\n}\r\nCurColor = CurColor / ((int)DistanceSteps*(int)RadialSteps);\r\nreturn CurColor;\r\n}"
   desc="float3 CurColor = 0;\r\nfloat2 BaseUV = MaterialFloat2(ScreenAlignedPosition(Parameters.ScreenPosition).xy);\r\nfloat2 NewUV = BaseUV;\r\nfloat StepSize = Distance / (int) DistanceSteps;\r\nfloat CurDistance = 0;\r\nfloat2 CurOffset = 0;\r\nfloat TwoPi = 6.283185;\r\nfloat Substep = 0;\r\nfloat2 ScenePixels=View.BufferSizeAndInvSize.xy*BaseUV;\r\nScenePixels+=View.TemporalAAParams.r;\r\nfloat2 RandomSamp = ((uint)(ScenePixels.x) + 2 * (uint)(ScenePixels.y)) % 5;\r\nRandomSamp+=Texture2DSample(Tex,TexSampler,ScenePixels);\r\nRandomSamp/=5;\r\nRandomSamp-=0.5;\r\nTempAARotation*=RandomSamp;\r\nTempAADistance*=StepSize*RandomSamp;\r\n\r\n\r\nint i=0;\r\nif (DistanceSteps<1)\r\n{\r\nreturn DecodeSceneColorForMaterialNode(NewUV);\r\n}\r\nelse\r\n{\r\nfloat2 BilinearOffset = View.BufferSizeAndInvSize.zw * 0.5f;\r\nfloat2 BilinearMinBufferUV = ViewportUVToBufferUV(float2(0.0f, 0.0f)) + BilinearOffset;\r\nfloat2 BilinearMaxBufferUV = ViewportUVToBufferUV(float2(1.0f, 1.0f)) - BilinearOffset;\r\n//CurDistance += 0.5*StepSize;\r\nwhile ( i < (int) DistanceSteps)\r\n{\r\n\r\n//CurDistance+=StepSize;\r\nfor (int j = 0; j < (int) RadialSteps; j++)\r\n{\r\nCurOffset.x = cos(TwoPi*((TempAARotation+Substep) / RadialSteps));\r\nCurOffset.y = sin(TwoPi*((TempAARotation+Substep) / RadialSteps));\r\nCurOffset *=DistanceMask;\r\nNewUV.x = BaseUV.x + (CurOffset.x * (CurDistance+(RandomSamp*TempAADistance)));\r\nNewUV.y = BaseUV.y + (CurOffset.y * (CurDistance+(RandomSamp*TempAADistance)));\r\nNewUV = clamp(NewUV, BilinearMinBufferUV, BilinearMaxBufferUV);\r\nCurColor += DecodeSceneColorForMaterialNode(NewUV);\r\n//CurDistance+=(StepSize+(TempAADistance))/RadialSteps;\r\nSubstep++;\r\n}\r\nCurDistance+=StepSize;\r\nSubstep+=RadialOffset;\r\ni++;\r\n}\r\nCurColor = CurColor / ((int)DistanceSteps*(int)RadialSteps);\r\nreturn CurColor;\r\n}"
   showCode=true
   inputs [
    {
     inputName=Distance
     input {expression=&self:MaterialExpressionLinearInterpolate_2}
    }
    {
     inputName=DistanceSteps
     input {expression=&self:MaterialExpressionMaterialFunctionCall_2}
    }
    {
     inputName=RadialSteps
     input {expression=&self:MaterialExpressionScalarParameter_21}
    }
    {
     inputName=TempAARotation
     input {expression=&self:MaterialExpressionConstant_4}
    }
    {
     inputName=DistanceMask
     input {expression=&self:MaterialExpressionConstant_4}
    }
    {
     inputName=Tex
     input {expression=&self:MaterialExpressionTextureObject_5}
    }
    {
     inputName=TempAADistance
     input {expression=&self:MaterialExpressionConstant_4}
    }
    {
     inputName=RadialOffset
     input {expression=&self:MaterialExpressionScalarParameter_22}
    }
    {
     inputName=RefractionAmount
     input {expression=&self:MaterialExpressionMultiply_23}
    }
   ]
  }
  NamedRerouteDeclaration(None=@n36) {name=BlurredColor, nodeColor=<rgba 0.0547 1 0 1>}
  n37 = VectorParameter() {group=TransparentArmor, parameterName=SurfaceBaseColor, sortPriority=5, defaultValue=<rgba 0.4271 0.4566 0.6406 1>}
  n38 = StaticSwitchParameter(
   True=Multiply(A=NamedRerouteUsage(), B=ScalarParameter() {defaultValue=1, group=TransparentArmor, parameterName=BaseColorMult, sortPriority=3})
   False=@n37.RGB
  ) {group=TransparentArmor, parameterName=-UseOriginalBaseColor, sortPriority=5}
  n39 = Multiply(A=NamedRerouteUsage(), B=@n38)
  n40 = ScalarParameter() {defaultValue=0.5, group=TransparentArmor, parameterName=OpacityBoost_Decal, sortPriority=2}
  n41 = NamedRerouteUsage()
  n42 = StaticSwitchParameter(True=LinearInterpolate(A=@n39, B=@n38, Alpha=LinearInterpolate(B=@n40, Alpha=@n41)), False=@n39) {group=TransparentArmor, parameterName=-UseTextureMapforOpacity, sortPriority=2}
  n43 = VectorParameter() {group=TransparentArmor, parameterName=SurfaceBaseColor_LowEnd, sortPriority=5, defaultValue=<rgba 0.4271 0.4566 0.6406 1>}
  n44 = StaticSwitchParameter(
   True=Multiply(A=NamedRerouteUsage(), B=ScalarParameter() {defaultValue=1, group=TransparentArmor, parameterName=BaseColorMult_LowEnd, sortPriority=3})
   False=@n43.RGB
  ) {group=TransparentArmor, parameterName=-UseOriginalBaseColor, sortPriority=5}
  NamedRerouteDeclaration(None=FeatureLevelSwitch(Default=@n42, ES2=@n44, ES3_1=@n44, SM5=@n42, SM6=@n42)) {name=BaseColor, nodeColor=<rgba 1 0 0.375 1>}
  n45 = Max(A=SceneColor!(UVs).RGB)
  n46 = Multiply(A=@n45, B=ScalarParameter() {defaultValue=1, group=TransparentArmor, parameterName=Opacity_SceneColor, sortPriority=2})
  n47 = ScalarParameter() {defaultValue=0.5, group=TransparentArmor, parameterName=MobileOpacity, sortPriority=2}
  n48 = ScalarParameter() {group=TransparentArmor, parameterName=Opacity_SceneColor_Min, sortPriority=2}
  n49 = ScalarParameter() {defaultValue=1, group=TransparentArmor, parameterName=Opacity_SceneColor_Max, sortPriority=2}
  n50 = TextureSampleParameter2D() {
   group=TransparentArmor
   parameterName=OpacityMap
   samplerSource=SSM_Wrap_WorldGroupSettings
   samplerType=SAMPLERTYPE_Masks
   sortPriority=2
   texture=&T_Fortnite_Default_S
  }
  n51 = ChannelMaskParameter(None=@n50.RGB) {group=TransparentArmor, parameterName=OpacityChannel, sortPriority=2}
  n52 = StaticSwitchParameter(
   True=LinearInterpolate(
    A=Multiply(A=@n45, B=@n48)
    B=Multiply(A=@n45, B=@n49)
    Alpha=NamedRerouteDeclaration(None=@n51.R) {name=OpacityMap, nodeColor=<rgba 1 0 0.375 1>}
   )
   False=@n46
  ) {group=TransparentArmor, parameterName=-UseTextureMapforOpacity, sortPriority=2}
  NamedRerouteDeclaration(None=FeatureLevelSwitch(Default=@n46, ES2=@n47, ES3_1=@n47, SM5=@n52, SM6=@n52)) {name=Opacity, nodeColor=<rgba 0 0.6953 1 1>}
  NamedRerouteDeclaration(None=StaticSwitchParameter(True=NamedRerouteUsage(), False=ScalarParameter() {defaultValue=1, group=TransparentArmor, parameterName=Specular, sortPriority=1}) {group=TransparentArmor, parameterName=-UseOriginalSpecular, sortPriority=1}) {name=Specular, nodeColor=<rgba 1 0 0.8906 1>}
  ScalarParameter() {defaultValue=4, group=TransparentArmor, parameterName=SpiralBlur_Steps}
 ]}
