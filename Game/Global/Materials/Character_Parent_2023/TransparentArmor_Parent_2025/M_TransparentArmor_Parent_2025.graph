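# Parent translucent material for transparent character armor, with per-pixel surface lighting and expression-driven shading models, usable on skeletal, cloth, Nanite and Niagara meshes.
# The skin-disabled 2023 character PreFX function supplies the base and passes its Tech Art Mask alpha into MF_TransparentArmor.
# That result feeds the shared character PostFX function, which also takes further inputs directly from PreFX.
# All logic lives in these material functions, chained through material attributes.
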Material /Game/Global/Materials/Character_Parent_2023/TransparentArmor_Parent_2025/M_TransparentArmor_Parent_2025
graph {
 properties {
  bUseMaterialAttributes=true
  bUsedWithClothing=true
  bUsedWithMeshDeformer=true
  bUsedWithMorphTargets=true
  bUsedWithNanite=true
  bUsedWithNiagaraMeshParticles=true
  bUsedWithSkeletalMesh=true
  blendMode=BLEND_Translucent
  pixelDepthOffsetMode=PDOM_Legacy
  shadingModel=MSM_FromMaterialExpression
  translucencyLightingMode=TLM_SurfacePerPixelLighting
  usedShadingModels=Default Lit | Subsurface | Clear Coat | Subsurface Profile | Cloth
 }
 nodes [
  n1 = MaterialFunctionCall!(Overide Texture Set UVs, Override Tech Art Mask UVs, Override Basic Emissive UVs) {materialFunction=&/MF_CharacterParent_2023_PreFX_SkinDisabled}
  MaterialFunctionCall(
   Material=MaterialFunctionCall!(TechArtMask=@n1.Tech Art Mask Alpha, In=@n1.Tech Art Mask Alpha) {materialFunction=&/MF_TransparentArmor}.Result
   M=@n1.M
   Pre Skinned Local Position=@n1.M
   Distance Blend=@n1.M
   Fresnel=@n1.M
  ) {materialFunction=&MF_CharacterParent_2023_PostFX}
 ]
}
